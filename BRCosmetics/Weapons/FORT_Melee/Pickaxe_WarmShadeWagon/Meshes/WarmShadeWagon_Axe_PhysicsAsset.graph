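# Physics asset for the WarmShadeWagon two-handed pickaxe cosmetic mesh, giving the melee weapon a simple collision shape.
# It holds a single body on the Melee_TwoHanded_Handle bone, made of one capsule offset along the bone and rotated to lie along the weapon.
# There are no constraints, so the weapon behaves as one rigid piece with no articulated parts.
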PhysicsAsset /BRCosmetics/Weapons/FORT_Melee/Pickaxe_WarmShadeWagon/Meshes/WarmShadeWagon_Axe_PhysicsAsset
bodies {Melee_TwoHanded_Handle [{shapeType=Capsule, length=73.3683, radius=30.8326, center=<0.1451 -1.5894 32.9434>, rotation=<rot 0.9489 -90 0.0854>}]}
constraints []
